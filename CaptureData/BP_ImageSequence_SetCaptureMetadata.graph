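# Editor utility Blueprint with functions to set, show and clear Capture Metadata on selected assets, named for image sequences.
# Set maps each asset to its existing metadata or a newly constructed object, opens an editable metadata window, and writes back only if that returns true.
# Show gathers valid metadata into an array for the same window in read-only mode.
# Clear removes metadata from each selected asset directly.
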EditorUtilityBlueprint /CaptureData/BP_ImageSequence_SetCaptureMetadata
graphs [
 Clear Capture Metadata {nodes [
   Clear Capture Metadata() {is=title K2Node_FunctionEntry}
   > n1 = GetSelectedAssets()
   > n2 = For Each Loop(Array:Array of Object References=@n1) -> (Array Element:Object Reference)
   > ClearCaptureMetadata!(execute, InObject:Object Reference=@n2) -> (then)
  ]}
 EventGraph {links [], nodes []}
 Set Capture Metadata {nodes [
   Set Capture Metadata() {is=title K2Node_FunctionEntry}
   > n1 = GetSelectedAssets()
   > n2 = For Each Loop(Array=@n1)
   > n3 = GetCaptureMetadata!(execute, InObject:Object Reference=@n2) -> (then, ReturnValue:Capture Metadata Object Reference)
   n4 = IsValid(Object=@n3)
   > n5 = Branch(Condition=@n4)
   n6 = Construct Object from Class!(execute=@n5.else, Class=/Script/CaptureDataEditor.CaptureMetadata, self, CameraId:String) -> (then, ReturnValue:Capture Metadata Object Reference)
   Add!(
    execute=[@n5.then, @n6.then]
    TargetMap:Map of Object References to Capture Metadata Object References=Get CaptureMetadataMap() -> (CaptureMetadataMap:Map of Object References to Capture Metadata Object References)
    Key:"Object Reference (by ref)"=@n2
    Value:"Capture Metadata Object Reference (by ref)"=Select(Option 0:Capture Metadata Object Reference=@n6, Option 1:Capture Metadata Object Reference=@n3, Index=@n4) -> (ReturnValue:Capture Metadata Object Reference)
   ) -> (then) {is=title K2Node_CallFunction}
   n7 = Get CaptureMetadataMap()
   n8 = Values!(execute=@n2.Completed, TargetMap:Map of Object References to Capture Metadata Object References=@n7) -> (then, Values:Array of Capture Metadata Object References)
   > n9 = ShowCaptureMetadataObjects!(
    execute
    InTitle:"Text (by ref)"=MakeLiteralText(Value="LOCTEXT(\"Capture Metadata\")")
    InObjects:Array of Object References=@n8
    InOptions:"Capture Metadata Window Options Structure (by ref)"=Make Capture Metadata Window Options!(bAllowEdit:bool=True) -> (CaptureMetadataWindowOptions:Capture Metadata Window Options Structure) {is=title K2Node_MakeStruct}
   ) -> (then, ReturnValue:bool)
   > Branch(Condition=@n9)
   > n10 = "For Each Loop (Map)"!(execute:Exec, MapPin:Map of Object References to Capture Metadata Object References=@n7, BreakPin:Exec) -> (then:Exec, KeyPin:Object Reference, ValuePin:Capture Metadata Object Reference, CompletedPin:Exec)
   > SetCaptureMetadata!(execute, InObject:Object Reference=@n10, InCaptureMetadata:Capture Metadata Object Reference=@n10.ValuePin) -> (then)
  ]}
 Show Capture Metadata {nodes [
   Show Capture Metadata() {is=title K2Node_FunctionEntry}
   > n1 = GetSelectedAssets()
   > n2 = For Each Loop(Array=@n1)
   n3 = Get CaptureMetadataArray() -> (CaptureMetadataArray:Array of Capture Metadata Object References)
   ShowCaptureMetadataObjects(
    execute=@n2.Completed
    InTitle=MakeLiteralText(Value="LOCTEXT(\"Capture Metadata\")")
    InObjects=@n3
    InOptions=Make Capture Metadata Window Options(bAllowEdit=false)
   )
   n4 = GetCaptureMetadata(execute=@n2.LoopBody, InObject=@n2)
   > Is Valid(InputObject=@n4)
   > Add(TargetArray:Array of Capture Metadata Object References=@n3, NewItem:"Capture Metadata Object Reference (by ref)"=@n4) {is=title K2Node_CallArrayFunction}
  ]}
]
variables []
