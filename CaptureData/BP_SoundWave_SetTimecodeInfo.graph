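# Editor utility that lets a user view and edit the timecode and frame rate of a selected SoundWave asset.
# SetTimecodeInfo takes the first selected SoundWave and reads its current values through GetTimecodeInfo.
# It copies them into a transient CaptureTimecodeInfo object shown in a modal details dialog, writing the edited values back only on confirmation.
# Despite its plural name, GetSelectedSoundWaves returns just one asset, so multi-selection is not batch processed.
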EditorUtilityBlueprint /CaptureData/BP_SoundWave_SetTimecodeInfo
graphs [
 EventGraph {links [], nodes []}
 GetSelectedSoundWaves {nodes [
   GetSelectedSoundWaves() {is=title K2Node_FunctionEntry}
   > n1 = GetSelectedAssetsOfClass(AssetClass=/Script/Engine.SoundWave)
   > n2 = Cast To SoundWave!(execute, Object="Get (a copy)"(Array:Array of Object References=@n1) -> (Output:Object Reference)) -> (then, CastFailed, AsSound Wave:Sound Wave Object Reference) {is=title K2Node_DynamicCast}
   > Return Node!(execute, Sound Wave:Sound Wave Object Reference=@n2)
  ]}
 GetTimecodeInfo {nodes [
   n1 = GetTimecodeInfo!() -> (then, Sound Wave:Sound Wave Object Reference) {is=title K2Node_FunctionEntry}
   > n2 = GetTimecode!(execute, InSoundWave:Sound Wave Object Reference=@n1) -> (then, ReturnValue:Timecode Structure)
   > n3 = GetFrameRate!(execute, InSoundWave:Sound Wave Object Reference=@n1) -> (then, ReturnValue:Frame Rate Structure)
   > Return Node!(execute, Timecode:Timecode Structure=@n2, Frame Rate:Frame Rate Structure=@n3)
  ]}
 SetTimecodeInfo {nodes [
   SetTimecodeInfo() {is=title K2Node_FunctionEntry}
   > n1 = GetSelectedSoundWaves!(execute, self) -> (then, Sound Wave:Sound Wave Object Reference) {is=title K2Node_CallFunction}
   > n2 = GetTimecodeInfo!(execute, self, Sound Wave:Sound Wave Object Reference=@n1) -> (then, Timecode:Timecode Structure, Frame Rate:Frame Rate Structure) {is=title K2Node_CallFunction}
   > n3 = Construct Object from Class!(
    execute
    Class=/Script/CaptureDataUtils.CaptureTimecodeInfo
    self
    Timecode:Timecode Structure=@n2
    FrameRate:Frame Rate Structure=@n2.Frame Rate
   ) -> (then, ReturnValue:Capture Timecode Info Object Reference)
   > n4 = Show Object Dialog!(
    execute
    Title:"Text (by ref)"=MakeLiteralText(Value="LOCTEXT(\"Set Timecode Info\")")
    InOutObject:Object Reference=@n3
    Options:"Editor Dialog Library Object Details View Options Structure (by ref)"=Make Editor Dialog Library Object Details View Options!(
     bShowObjectName:bool=false
     bAllowSearch:bool=false
     bAllowResizing:bool=true
     MinWidth:int=400
     MinHeight:int=200
     ValueColumnWidthRatio:float=0.6
    ) -> (EditorDialogLibraryObjectDetailsViewOptions:Editor Dialog Library Object Details View Options Structure) {is=title K2Node_MakeStruct}
   ) -> (then, ReturnValue:bool)
   > Branch(Condition=@n4)
   > SetTimecodeInfo!(
    execute
    InTimecode:"Timecode Structure (by ref)"=Get Timecode!(self:Capture Timecode Info Object Reference=@n3) -> (Timecode:Timecode Structure)
    InFrameRate:"Frame Rate Structure (by ref)"=Get FrameRate!(self:Capture Timecode Info Object Reference=@n3) -> (FrameRate:Frame Rate Structure)
    OutSoundWave:Sound Wave Object Reference=@n1
   ) -> (then) {is=title K2Node_CallFunction}
  ]}
]
variables []
